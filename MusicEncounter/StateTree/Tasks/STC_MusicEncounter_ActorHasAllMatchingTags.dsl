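; StateTree condition for the Music Encounter system that tests whether the context Actor has all of a configured set of gameplay tags.
; It queries the Actor through the gameplay tag interface message HasAllMatchingGameplayTags and negates the result when Invert is set.
; Its editor description switches between "Actor Has ALL Matching Tags" and "Actor Does Not Have ALL Matching Tags" to match the Invert setting.
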(fn GetDescription
  (params (Formatting :type EStateTreeNodeFormatting%20Enum :default "RichText"))
  (return (Utilities|String|SelectText "NSLOCTEXT(\"[69D4078C30B434D5AA1208E71C781736]\", \"EFA4B347439F386584868DAC2C639E15\", \"Actor Does Not Have ALL Matching Tags\")" "NSLOCTEXT(\"[69D4078C30B434D5AA1208E71C781736]\", \"DFD129144EAF93D6DC1AEAAD1F23162D\", \"Actor Has ALL Matching Tags\")" (Variables|Default|GetInvert))))

(fn ReceiveTestCondition
  (bind _returnvalue (GameplayTags|HasAllMatchingGameplayTags(Message) (Variables|Context|GetActor) (Variables|Default|GetGameplay%20Tags%20to%20Check)))
  (return (select (Variables|Default|GetInvert) (not _returnvalue) _returnvalue)))
