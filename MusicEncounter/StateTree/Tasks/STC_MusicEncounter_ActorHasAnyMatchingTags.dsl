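; StateTree condition for music encounters that tests whether the context actor has any of a configurable set of gameplay tags.
; It queries the actor through the gameplay tag interface message for a match against the tags to check, and an Invert flag flips the result.
; The editor description switches between "Has" and "Does Not Have" wording to reflect that flag.
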(fn GetDescription
  (params (Formatting :type EStateTreeNodeFormatting%20Enum :default "RichText"))
  (return (Utilities|String|SelectText "NSLOCTEXT(\"[2E9A9106B0DF3ED02862283752F19445]\", \"8E1A9FC941C64413F0BC4DA4E42FE40B\", \"Actor Does Not Have ANY Matching Tags\")" "NSLOCTEXT(\"[2E9A9106B0DF3ED02862283752F19445]\", \"B52B5F1844E23CD3713FD59ABE8C3611\", \"Actor Has ANY Matching Tags\")" (Variables|Default|GetInvert))))

(fn ReceiveTestCondition
  (bind _returnvalue (GameplayTags|HasAnyMatchingGameplayTags(Message) (Variables|Context|GetActor) (Variables|Default|GetGameplay%20Tags%20to%20Check)))
  (return (select (Variables|Default|GetInvert) (not _returnvalue) _returnvalue)))
