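; StateTree task that relays a music encounter boss NPC's alert level changes to the StateTree.
; On enter, it casts the context actor to the non-participant boss pawn, stores it as the Boss Pawn output and, if that pawn uses alertness, binds to its alert level change event.
; The handler broadcasts a delegate selected by the new level, with Unaware and Threatened delegates wired.
; On exit, it unbinds the handler.
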(event EventEnterState
  (params (Transition :type State%20Tree%20Transition%20Result%20Structure%20%28by%20ref%29))
  (bind _asbp_player_pawn_non_participant_boss (Utilities|Casting|CastToBP_PlayerPawn_NonParticipant_Boss (Variables|Context|GetActor)))
  (bind _output_get (Variables|Output|SetBoss%20Pawn _asbp_player_pawn_non_participant_boss))
  (Utilities|IsValid _output_get
    (:Is%20Valid
      (Utilities|FlowControl|Sequence
        (:then_0
          (if (Variables|/Game/Athena/AI/NPCs/Base/BP_PlayerPawn_NonParticipant.BP_PlayerPawn_NonParticipant_C|GetUseAlertness (Variables|Output|GetBoss%20Pawn))
            (Default|BindEventtoOnAlertLevelChanged (Variables|Output|GetBoss%20Pawn) (EventDispatchers|CreateEvent :!event Custom|Handle%20Alert%20Level%20Changed))))))))

(event EventExitState
  (params (Transition :type State%20Tree%20Transition%20Result%20Structure%20%28by%20ref%29))
  (bind _boss_pawn (Variables|Output|GetBoss%20Pawn)
    (:then
      (Default|UnbindEventfromOnAlertLevelChanged _boss_pawn (EventDispatchers|CreateEvent :!event Custom|Handle%20Alert%20Level%20Changed)))
    (:else)))

(event Custom|Handle%20Alert%20Level%20Changed
  (params (OldAlertLevel :type EAlertLevel%20Enum :default "Unaware") (NewAlertLevel :type EAlertLevel%20Enum :default "Unaware"))
  (StateTree|BroadcastDelegate (Utilities|Select :Index NewAlertLevel :Unaware (Variables|Output|GetUnaware) :Threatened (Variables|Output|GetThreatened))))
