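; StateTree task for the music encounter system that starts or resumes a 2D music track on entering a state.
; If the output audio component is already playing, it unpauses it through the MusicEncounter.bIsPaused parameter; otherwise it spawns the input music and stores the component.
; Exit sets that parameter to pause rather than stopping playback.
; Without cosmetic support the task succeeds immediately; missing music fails it when a spawn is needed.
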(event EventEnterState
  (params (Transition :type State%20Tree%20Transition%20Result%20Structure%20%28by%20ref%29))
  (Networking|CanExecuteCosmeticEvents
    (:True
      (bind _audio_component (Variables|Output|GetAudio%20Component)
        (:then
          (if (Audio|Components|Audio|IsPlaying _audio_component)
            (Audio|Parameter|SetBooleanParameter (Variables|Output|GetAudio%20Component) "MusicEncounter.bIsPaused" false :!class /Script/AudioExtensions.AudioParameterControllerInterface)
            (StateTree|FinishTask true)
            (else
              _)))
        (:else
          _))
      (bind _music (Variables|Input|GetMusic)
        (:then
          (bind _returnvalue (Audio|SpawnSound2D _music))
          (Variables|Output|SetAudio%20Component _returnvalue)
          (StateTree|FinishTask true))
        (:else
          (StateTree|FinishTask false))))
    (:False
      (StateTree|FinishTask true))))

(event EventExitState
  (params (Transition :type State%20Tree%20Transition%20Result%20Structure%20%28by%20ref%29))
  (bind _audio_component (Variables|Output|GetAudio%20Component)
    (:then
      (Audio|Parameter|SetBooleanParameter _audio_component "MusicEncounter.bIsPaused" true :!class /Script/AudioExtensions.AudioParameterControllerInterface))
    (:else)))
