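; StateTree task that spawns a music encounter Audio Gameplay Volume for the context actor while its state is active.
; On enter, it spawns the parameterized class, outputs it, sets radius and cylinder half-height through BPI_MusicEncounterAGVInterface, and attaches it to the actor.
; It relays the volume's toggled on/off events as Listener Entered/Exited StateTree delegates.
; On exit, it destroys the spawned volume.
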(event EventEnterState
  (params (Transition :type State%20Tree%20Transition%20Result%20Structure%20%28by%20ref%29))
  (bind _returnvalue (Game|SpawnActorfromClass (Variables|Parameters|GetVolume%20Class) (Math|Transform|MakeTransform "0, 0, 0") "Undefined" "MultiplyWithRoot" (Variables|Context|GetActor) 0 0))
  (bind _output_get (Variables|Output|SetMusicEncounter%20AGV _returnvalue))
  (Utilities|IsValid _output_get
    (:Is%20Valid
      (Utilities|FlowControl|Sequence
        (:then_0
          (bind _audiogameplayvolume (Variables|/MusicEncounter/Actors/AudioGameplayVolumes/BP_MusicEncounterAudioGameplayVolume.BP_MusicEncounterAudioGameplayVolume_C|GetAudioGameplayVolume (Variables|Output|GetMusicEncounter%20AGV))
            (:then
              (Events|BindEventtoOnToggledOn _audiogameplayvolume (EventDispatchers|CreateEvent :!event Custom|Handle%20Listener%20Entered))
              (Events|BindEventtoOnToggledOff _audiogameplayvolume (EventDispatchers|CreateEvent :!event Custom|Handle%20Listener%20Exited)))
            (:else)))
        (:then_1
          (Utilities|FlowControl|Sequence
            (:then_0
              (MusicEncounter|Volume|SetRadius (Variables|Output|GetMusicEncounter%20AGV) (Variables|Parameters|GetRadius) :!class /MusicEncounter/Actors/AudioGameplayVolumes/BPI_MusicEncounterAGVInterface.BPI_MusicEncounterAGVInterface_C)
              (MusicEncounter|Volume|SetCylinderHalfHeight (Variables|Output|GetMusicEncounter%20AGV) (Variables|Parameters|GetHalf%20Height%20%28Cylinder%29) :!class /MusicEncounter/Actors/AudioGameplayVolumes/BPI_MusicEncounterAGVInterface.BPI_MusicEncounterAGVInterface_C))
            (:then_1
              (Utilities|FlowControl|Sequence
                (:then_0
                  (Transformation|AttachActorToActor (Variables|Output|GetMusicEncounter%20AGV) (Variables|Context|GetActor) "None" "SnapToTarget" "SnapToTarget" "SnapToTarget" true))))))))))

(event EventExitState
  (params (Transition :type State%20Tree%20Transition%20Result%20Structure%20%28by%20ref%29))
  (bind _musicencounter_agv (Variables|Output|GetMusicEncounter%20AGV)
    (:then
      (Actor|DestroyActor _musicencounter_agv))
    (:else)))

(event Custom|Handle%20Listener%20Entered
  (StateTree|BroadcastDelegate (Variables|Output|GetListener%20Entered)))

(event Custom|Handle%20Listener%20Exited
  (StateTree|BroadcastDelegate (Variables|Output|GetListener%20Exited)))
